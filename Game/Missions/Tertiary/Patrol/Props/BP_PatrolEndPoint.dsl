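; Patrol mission end point where AI from its encounter despawn through death rifts.
; Awake encounter AI entering the proximity box spawn rifts, up to a cap, spread around the despawn box and ground-traced.
; In the despawn box, non-Shielder AI targeting this actor get a server teleport-out event and cosmetic rift suck-in; player-lured AI are polled by timer until they retarget it or leave.
; Destruction deactivates the locally spawned rifts.
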(event OnComponentBeginOverlap(ProximityBox)
  (params (OverlappedComponent :type Primitive%20Component%20Object%20Reference) (OtherActor :type Actor%20Object%20Reference) (OtherComp :type Primitive%20Component%20Object%20Reference) (OtherBodyIndex :type Integer) (bFromSweep :type Boolean) (SweepResult :type Hit%20Result%20Structure%20%28by%20ref%29))
  (bind _encounter (Variables|Default|GetEncounter))
  ;[1 1 1] When AI from the Encounter get close, enable the rift FX.
  (Utilities|Casting|CastToCapsuleComponent OtherComp)
  (bind _asfort_aipawn (Utilities|Casting|CastToFortAIPawn OtherActor))
  (bind _returnvalue (AI|GetSleepComponent _asfort_aipawn))
  (Utilities|IsValid _returnvalue
    (:Is%20Valid
      (if (and (not (AI|IsSleeping _returnvalue)) (Utilities|Equal(Object) (Encounter|GetEncounterInfo _asfort_aipawn :!class /Script/FortniteGame.FortPawn) _encounter))
        _))
    (:Is%20Not%20Valid
      (if (Utilities|Equal(Object) (Encounter|GetEncounterInfo _asfort_aipawn :!class /Script/FortniteGame.FortPawn) _encounter)
        _))) ; If sleep component isn't valid, only check encounter info
  (CallFunction|SpawnRift))

(event Custom|CloseRift
  ;[1 1 1] Destroy self after being closed.
  (Actor|DestroyActor self))

(event EventDestroyed
  ;[1 1 1] Make sure when this patrol end point dies that it kills all of the death rifts it spawned.  The rifts were created on local clients and must be destroyed there.  We may not have received the replicated CleanupRift flag yet.
  (CallFunction|DeactivateAllDeathRifts))

(event Custom|DespawnEnemy
  (params (Pawn :type Fort%20AIPawn%20Object%20Reference))
  (bind _riftrandom (Variables|Default|GetRiftRandom))
  (bind _deathrifts (Variables|Default|GetDeathRifts))
  (bind _output (Utilities|Array|Get(acopy) _deathrifts _riftrandom))
  ;[1 1 1] Client Despawn Visuals Logic
  (Networking|CanExecuteCosmeticEvents
    (:True
      (Variables|Default|SetRiftRandom (Math|Random|RandomIntegerinRange 0 (Utilities|Array|LastIndex (Variables|Default|GetDeathRifts))))
      (Utilities|IsValid _output
        (:Is%20Valid
          (Class|BRiftPortals|PatrolEnemySuckIn _output 0.8 1.25 false)
          (bind _asenemy_pawn_parent (Utilities|Casting|CastToEnemyPawn_Parent Pawn))
          (Class|EnemyPawnParent|DespawnEnemy _asenemy_pawn_parent (+ (Transformation|GetActorLocation _output) "0, 0,125.000000")))))))

(event OnComponentBeginOverlap(DespawnBox)
  (params (OverlappedComponent :type Primitive%20Component%20Object%20Reference) (OtherActor :type Actor%20Object%20Reference) (OtherComp :type Primitive%20Component%20Object%20Reference) (OtherBodyIndex :type Integer) (bFromSweep :type Boolean) (SweepResult :type Hit%20Result%20Structure%20%28by%20ref%29))
  (bind _aitouchingdespawntrigger (Variables|Default|GetAITouchingDespawnTrigger))
  (bind _returnvalue (Encounter|GetEncounterInfo _aitouchingdespawntrigger :!class /Script/FortniteGame.FortPawn))
  (bind _encounter (Variables|Default|GetEncounter))
  (bind _returnvalue_2 (Utilities|Equal(Object) _returnvalue _encounter))
  (bind _aitouchingdespawntrigger_2 (Variables|Default|GetAITouchingDespawnTrigger))
  ;[1 1 1] Despawn AI from the Encounter when they reach the rift.
  (bind _returnvalue_3 (GameplayTags|HasMatchingGameplayTag(Message) OtherActor "(TagName=\"NPC.CharacterType.Shielder.Basic\")"))
  (if (not _returnvalue_3)
    (Utilities|Casting|CastToCapsuleComponent OtherComp)
    (bind _asfort_aipawn (Utilities|Casting|CastToFortAIPawn OtherActor))
    (bind _output_get (Variables|Default|SetAITouchingDespawnTrigger _asfort_aipawn))
    (bind _returnvalue_4 (AI|GetSleepComponent _output_get))
    (Utilities|IsValid _returnvalue_4
      (:Is%20Valid
        (if (and (not (AI|IsSleeping _returnvalue_4)) _returnvalue_2)
          _))
      (:Is%20Not%20Valid
        (if _returnvalue_2
          _))) ; If sleep component isn't valid, only check encounter info
    (Networking|SwitchHasAuthority
      (:Authority
        (if (Utilities|Equal(Object) (AI|Components|Blackboard|GetValueasObject (AI|GetBlackboard _aitouchingdespawntrigger) (Utilities|Name|MakeLiteralName "GoalActor")) self)
          (Ability|SendGameplayEventtoActor _aitouchingdespawntrigger_2 (Variables|Default|GetTC_NPCAbilityNonAttackTeleportOut) (Utilities|Struct|MakeGameplayEventData :Instigator (Variables|Default|GetAITouchingDespawnTrigger) :Target self :ContextHandle "()" :InstigatorTags "(GameplayTags=)" :TargetTags "(GameplayTags=)" :EventMagnitude 0.0 :TargetData "()"))
          (CallFunction|DespawnEnemy _aitouchingdespawntrigger_2)
          (else
            (CallFunction|AILuredIntoEndPointbyPlayer)))))))

(event Custom|AILuredIntoEndPointByPlayer
  ;[1 1 1] Manage AI trolled into the End Point by Players. If the Player dies, make sure these AI get Despawned.
  (Utilities|Array|Add (Variables|Default|GetAILuredByPlayerArray) (Variables|Default|GetAITouchingDespawnTrigger))
  (if (not (Utilities|Time|IsTimerActivebyFunctionName 0 "LuredAICleanup"))
    (Utilities|Time|SetTimerbyFunctionName 0 "LuredAICleanup" 1 true)))

(event Custom|LuredAICleanup
  ;[1 1 1] Manage AI trolled into the End Point by Players. If the Player dies, make sure these AI get Despawned.
  (for _array_element (Variables|Default|GetAILuredByPlayerArray)
    (Networking|SwitchHasAuthority
      (:Authority
        (if (Utilities|Equal(Object) (AI|Components|Blackboard|GetValueasObject (AI|GetBlackboard _array_element) (Utilities|Name|MakeLiteralName "GoalActor")) self)
          (Ability|SendGameplayEventtoActor _array_element (Variables|Default|GetTC_NPCAbilityNonAttackTeleportOut) (Utilities|Struct|MakeGameplayEventData :Instigator _array_element :Target self :ContextHandle "()" :InstigatorTags "(GameplayTags=)" :TargetTags "(GameplayTags=)" :EventMagnitude 0.0 :TargetData "()")) ; Despawn dis MADAFAKA
          (Utilities|Array|RemoveItem (Variables|Default|GetAILuredByPlayerArray) _array_element)))))
  (if (== (Utilities|Array|Length (Variables|Default|GetAILuredByPlayerArray)) 0)
    (Utilities|Time|ClearTimerbyFunctionName 0 "LuredAICleanup")))

(event OnComponentEndOverlap(DespawnBox)
  (params (OverlappedComponent :type Primitive%20Component%20Object%20Reference) (OtherActor :type Actor%20Object%20Reference) (OtherComp :type Primitive%20Component%20Object%20Reference) (OtherBodyIndex :type Integer))
  ;[1 1 1] Manage AI trolled into the End Point by Players. If the Player dies, make sure these AI get Despawned.
  (bind _asfort_aipawn (Utilities|Casting|CastToFortAIPawn OtherActor))
  (Utilities|Array|RemoveItem (Variables|Default|GetAILuredByPlayerArray) _asfort_aipawn)
  (if (== (Utilities|Array|Length (Variables|Default|GetAILuredByPlayerArray)) 0)
    (Utilities|Time|ClearTimerbyFunctionName 0 "LuredAICleanup")))

(fn DeactivateAllDeathRifts
  (for _array_element (Variables|Default|GetDeathRifts)
    (Class|BRiftPortals|ManuallyDeactivateRift _array_element))
  (Utilities|Array|Clear (Variables|Default|GetDeathRifts)))

(fn OnRep_PatrolEndPointLocation
  (Transformation|SetActorLocationAndRotation self (Variables|Default|GetPatrolEndPointLocation) "0, 0, 0" false false))

(fn SpawnRift
  (bind _deathrifts (Variables|Default|GetDeathRifts))
  (bind _returnvalue (Utilities|Array|Length _deathrifts))
  (bind _numofexistingrifts (LocalVariables|GetNumOfExistingRifts))
  (bind _returnvalue_2 (Math|Conversions|ToFloat(Integer) _numofexistingrifts))
  (bind _despawnbox (Variables|Default|GetDespawnBox))
  (bind _returnvalue_3 (Transformation|GetWorldLocation _despawnbox))
  (bind _riftsspawndiameter (Variables|Default|GetRiftsSpawnDiameter))
  (bind _pi_2_x (LocalVariables|Getpi*2/x))
  (bind _returnvalue_4 (* _pi_2_x _returnvalue_2))
  (bind _returnvalue_5 (Math|Trig|Cos(Radians) _returnvalue_4))
  (bind _returnvalue_6 (* _returnvalue_5 _riftsspawndiameter))
  (bind _riftsspawndiameter_2 (Variables|Default|GetRiftsSpawnDiameter))
  (bind _returnvalue_7 (Math|Trig|Sin(Radians) _returnvalue_4))
  (bind _returnvalue_8 (* _returnvalue_7 _riftsspawndiameter_2))
  (bind _returnvalue_9 (Math|Vector|MakeVector _returnvalue_8 _returnvalue_6))
  (bind _returnvalue_10 (Math|Random|RandomFloatinRange 0.25 1))
  (bind _returnvalue_11 (Math|Vector|Lerp(Vector) "0, 0, 0" _returnvalue_9 _returnvalue_10))
  (bind _returnvalue_12 (+ _returnvalue_3 _returnvalue_11))
  (if (not (>= _returnvalue (Variables|Default|GetNumberOfRiftsToSpawn)))
    (LocalVariables|SetNumOfExistingRifts _returnvalue)
    (LocalVariables|Setpi*2/x (/ (* (Math|Trig|GetPI) 2) (Math|Conversions|ToFloat(Integer) (Variables|Default|GetNumberOfRiftsToSpawn))))
    (bind (_outhit _returnvalue_13) (Collision|LineTraceByChannel _returnvalue_12 (- _returnvalue_12 "0.000000,0.000000,512.000000")))
    (bind _returnvalue_14 (Game|SpawnActorfromClass "/Game/VisualThreatManager/Rifts/Blueprints/B_Rift_Portals.B_Rift_Portals_C" (Math|Transform|MakeTransform (select _returnvalue_13 (+ (.Location (Collision|BreakHitResult _outhit)) "0, 0,10.000000") (- _returnvalue_12 "0, 0,100.000000"))) "Undefined" "MultiplyWithRoot" 0 1.5 3.0 0 0))
    (Class|BRiftPortals|ManuallyActivateRift _returnvalue_14)
    (Utilities|Array|Add (Variables|Default|GetDeathRifts) _returnvalue_14)))

(fn ConstructionScript)
